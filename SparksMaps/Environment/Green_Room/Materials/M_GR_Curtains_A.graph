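# M_GR_Curtains_A is a two-sided, Default Lit surface material for the curtains in the Green Room environment, flagged for use with Nanite.
# It declares three texture parameters, Difuse, Normal and a linear-color Masks map, each defaulting to a curtain texture and open to override in material instances.
# The graph holds only these samplers, with no math nodes and no visible connections to material outputs.
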Material /SparksMaps/Environment/Green_Room/Materials/M_GR_Curtains_A
graph {
 properties {bUsedWithNanite=true, pixelDepthOffsetMode=PDOM_Legacy, twoSided=true, usedShadingModels=Default Lit}
 nodes [
  TextureSampleParameter2D() {parameterName=Difuse, texture=/SparksMaps/Environment/Green_Room/Textures/T_GR_Curtains_D}
  TextureSampleParameter2D() {parameterName=Normal, samplerType=SAMPLERTYPE_Normal, texture=/SparksMaps/Environment/Green_Room/Textures/T_GR_Curtains_N}
  TextureSampleParameter2D() {parameterName=Masks, samplerType=SAMPLERTYPE_LinearColor, texture=/SparksMaps/Environment/Green_Room/Textures/T_GR_Curtains_S}
 ]
}
