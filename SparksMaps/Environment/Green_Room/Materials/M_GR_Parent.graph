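# Parent surface material for the Green Room environment, a Default Lit shader configured by static switches and usable with Nanite and instanced meshes.
# Base color, normals and PBR masks support tinting, detail normals, selectable UV sets and scalar overrides.
# An optional Fresnel layer changes color, metallic and roughness.
# Optional emissive uses albedo, mask or UV1 gradient, with stage-palette color and Fresnel options, then passes through beat-pulse and FX-adjustment functions.
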Material /SparksMaps/Environment/Green_Room/Materials/M_GR_Parent
graph {
 properties {
  bUsedWithInstancedStaticMeshes=true
  bUsedWithNanite=true
  physMaterial=/Game/FortressPhysicalMaterials/Metal
  pixelDepthOffsetMode=PDOM_Legacy
  usedShadingModels=Default Lit
 }
 nodes [
  n1 = ScalarParameter() {defaultValue=1, group=Emissive, parameterName=Emissive_Albedo_Intensity, sliderMax=20}
  n2 = TextureSampleParameter2D(UVs=NamedRerouteUsage()) {parameterName=Diffuse, samplerType=SAMPLERTYPE_Masks, texture=/Game/Global/Textures/Default/Blanks/T_MidgreyMask}
  n3 = TextureSampleParameter2D() {parameterName=Emiss, samplerType=SAMPLERTYPE_Masks, texture=&T_MidgreyMask}
  n4 = MaterialFunctionCall!(CustomUPosition) {materialFunction=/SparksStageShow/Systems/StageManager/FXManager/MaterialFunctions/MF_MojoStageColors}
  n5 = VectorParameter() {group=Emissive, parameterName=Emissive_Tint, sortPriority=34, defaultValue=<rgba 1 1 1 1>}
  n6 = Multiply(A=@n5.RGB, B=@n5.RGB)
  n7 = StaticBoolParameter() {parameterName=UseStagePalette}
  n8 = ComponentMask(None=TextureCoordinate() {coordinateIndex=1}) {g=true}
  n9 = ScalarParameter() {defaultValue=2, group=Emissive, parameterName=GradientPower}
  n10 = StaticSwitch(
   True=Multiply(A=StaticSwitch(True=@n4.Primary, False=@n6, Value=@n7), B=@n3.RGB)
   False=LinearInterpolate(A=Constant3Vector(), B=@n6, Alpha=Power(Base=@n8, Exp=@n9))
   Value=StaticBoolParameter() {group=Emissive, parameterName=UseEmissMask, sortPriority=33}
  )
  n11 = MaterialFunctionCall(
   In=Fresnel(
    ExponentIn=ScalarParameter() {defaultValue=0.06, group=Emissive, parameterName=Fresnel_Strength, sortPriority=42}
    BaseReflectFractionIn=ScalarParameter() {defaultValue=50, group=Emissive, parameterName=Fresnel_Falloff, sortPriority=43}
   ) {baseReflectFraction=0.003}
   Contrast=ScalarParameter() {defaultValue=0.03, group=Emissive, parameterName=Fresnel_Contrast, sortPriority=44}
  ) {materialFunction=&CheapContrast}
  n12 = StaticSwitch(
   True=OneMinus(None=@n11.Result)
   False=@n11.Result
   Value=StaticBoolParameter() {group=Emissive, parameterName=Fresnel_Inner_Glow, sortPriority=41}
  )
  n13 = StaticSwitch(
   True=Multiply(A=Multiply(A=@n1, B=@n2.RGB), B=@n3.RGB)
   False=StaticSwitch(
    True=Multiply(A=@n10, B=@n12)
    False=@n10
    Value=StaticBoolParameter() {group=Emissive, parameterName=UseFresnel_with_Emissive, sortPriority=40}
   )
   Value=StaticBoolParameter() {group=Emissive, parameterName=UseAlbedoEmissive, sortPriority=33}
  )
  n14 = Constant()
  n15 = StaticBoolParameter() {group=Emissive, parameterName=UseEmissive}
  MaterialFunctionCall(Emissive=MaterialFunctionCall!(Emissive=StaticSwitch(True=@n13, False=@n14, Value=@n15), BeatPulse Strength, BeatPulse Min, BeatPulse Max) {materialFunction=/SparksMaps/Environment/MaterialFunctions/MF_MojoBeatPulse}.Result) {materialFunction=&MF_FX_EmissiveAdjustments}
  MaterialFunctionCall() {materialFunction=/SparksMaps/Environment/Shared/Materials/Functions/MF_Sparks_WPO}
  n16 = TextureSampleParameter2D(UVs=NamedRerouteUsage()) {parameterName=Normals, samplerType=SAMPLERTYPE_Normal, texture=&FlatNormalBlue}
  n17 = TextureSampleParameter2D(UVs=MaterialFunctionCall!(
    Tile U=ScalarParameter() {group=Normal, parameterName=Detail Normal U}
    Tile V=ScalarParameter() {group=Normal, parameterName=Detail Normal V}
    UVSet=ScalarParameter() {group=Normal, parameterName=Detail Normal UV Set}
   ) {materialFunction=/SparksMaps/Environment/Shared/Materials/Functions/MF_Choose_UV_Set}.Result) {group=Normal, parameterName=Detail Normals, samplerType=SAMPLERTYPE_Normal, texture=&FlatNormalBlue}
  MaterialFunctionCall(
   Normal=StaticSwitchParameter(
    True=MaterialFunctionCall(BaseNormal=@n16.RGB, AdditionalNormal=@n17.RGB) {materialFunction=&BlendAngleCorrectedNormals}.Result
    False=@n16.RGB
   ) {group=Normal, parameterName=Use Detail Normal}
   Flatness=ScalarParameter() {group=Normal, parameterName=Normal intensity}
  ) {materialFunction=&FlattenNormal}
  n18 = ScalarParameter() {defaultValue=1, parameterName=Tile_U}
  n19 = ScalarParameter() {defaultValue=1, parameterName=Tile_V}
  NamedRerouteDeclaration(None=Multiply(A=TextureCoordinate(), B=AppendVector(A=@n18, B=@n19))) {name=GlobalUVControl, nodeColor=<rgba 0.9688 1 0 1>}
  n20 = TextureSampleParameter2D(UVs=NamedRerouteUsage()) {
   parameterName=PBR Masks
   samplerType=SAMPLERTYPE_Masks
   texture=&T_MidgreyMask
   channelNames {a=Unused, b=Roughness, g=Metallic, r=Specular}
  }
  Reroute(None=StaticSwitch(
    True=ScalarParameter() {defaultValue=0.5, group=Override_PBR, parameterName=Override Specular, sliderMax=1}
    False=@n20.Specular
    Value=StaticBoolParameter() {group=Override_PBR, parameterName=Override Specular, sortPriority=24}
   ))
  n21 = TextureSampleParameter2D(UVs=NamedRerouteUsage()) {
   group=Diffuse_Fresnel
   parameterName=Diffuse_Fresnel Orerride Masks
   samplerType=SAMPLERTYPE_Masks
   texture=&T_MidgreyMask
   channelNames {a=Unused, b=Roughness, g=Metallic, r=Specular}
  }
  n22 = StaticSwitch(
   True=ScalarParameter() {group=Override_PBR, parameterName=Override Metallic, sliderMax=1}
   False=@n20.Metallic
   Value=StaticBoolParameter() {group=Override_PBR, parameterName=Override Metallic, sortPriority=24}
  )
  n23 = LinearInterpolate(
   A=@n22
   B=ScalarParameter() {group=Diffuse_Fresnel, parameterName=Diffuse_Fresnel_Metallic, sliderMax=1, sortPriority=29}
   Alpha=NamedRerouteUsage()
  )
  n24 = StaticBoolParameter() {group=Diffuse_Fresnel, parameterName=Overide Fresnel metallic with Texture, sortPriority=24}
  StaticSwitch(
   True=StaticSwitch(
    True=StaticSwitch(True=@n21.Metallic, False=@n23, Value=@n24)
    False=@n22
    Value=StaticBoolParameter() {group=Override_PBR, parameterName=Override Metallic, sortPriority=24}
   )
   False=@n22
   Value=StaticBoolParameter() {group=Diffuse_Fresnel, parameterName=Use_Diffuse_Fresnel, sortPriority=5}
  )
  n25 = StaticSwitch(
   True=ScalarParameter() {defaultValue=0.5, group=Override_PBR, parameterName=Override Roughness, sliderMax=1}
   False=@n20.Roughness
   Value=StaticBoolParameter() {group=Override_PBR, parameterName=Override Roughness, sortPriority=24}
  )
  n26 = LinearInterpolate(
   A=@n25
   B=ScalarParameter() {defaultValue=0.5, group=Diffuse_Fresnel, parameterName=Diffuse_Fresnel_Roughness, sliderMax=1, sortPriority=28}
   Alpha=NamedRerouteUsage()
  )
  n27 = StaticBoolParameter() {group=Diffuse_Fresnel, parameterName=Overide Fresnel roughness with Texture, sortPriority=25}
  StaticSwitch(
   True=StaticSwitch(
    True=StaticSwitch(True=@n21.Roughness, False=@n26, Value=@n27)
    False=@n25
    Value=StaticBoolParameter() {group=Override_PBR, parameterName=Override Roughness, sortPriority=24}
   )
   False=@n25
   Value=StaticBoolParameter() {group=Diffuse_Fresnel, parameterName=Use_Diffuse_Fresnel, sortPriority=5}
  )
  n28 = TextureSampleParameter2D(UVs=MaterialFunctionCall(
    Tile U=ScalarParameter() {defaultValue=1, group=Diffuse_UV_Set, parameterName=Diffuse_Tile_U}
    Tile V=ScalarParameter() {defaultValue=1, group=Diffuse_UV_Set, parameterName=Diffuse_Tile_V}
    UVSet=ScalarParameter() {group=Diffuse_UV_Set, parameterName=Diffuse_UV_Set}
   ) {materialFunction=&MF_Choose_UV_Set}.Result) {
   group=Diffuse_UV_Set
   parameterName=Diffuse_Pattern_Texture
   samplerType=SAMPLERTYPE_Masks
   texture=&T_MidgreyMask
  }
  n29 = VectorParameter() {group=Tint, parameterName=BaseColor_Tint_Intensity, sortPriority=36, defaultValue=<rgba 1 1 1 1>}
  n30 = Multiply(A=@n29.RGB, B=@n2.RGB)
  n31 = StaticSwitchParameter(True=LinearInterpolate(A=@n2.RGB, B=Multiply(A=@n30, B=@n29.A), Alpha=@n2.RGB), False=@n2.RGB) {group=Tint, parameterName=Enable Tint, sortPriority=35}
  n32 = StaticSwitch(
   True=Multiply(A=@n28.RGB, B=@n31)
   False=@n31
   Value=StaticBoolParameter() {group=Diffuse_UV_Set, parameterName=Choose_UV_Diffuse_Multiply, sortPriority=20}
  )
  n33 = TextureSampleParameter2D(UVs=MaterialFunctionCall(
    Tile U=ScalarParameter() {defaultValue=1, group=Diffuse_Fresnel, parameterName=DiffuseFresnel_U}
    Tile V=ScalarParameter() {defaultValue=1, group=Diffuse_Fresnel, parameterName=DiffuseFresnel_V}
    UVSet=ScalarParameter() {group=Diffuse_Fresnel, parameterName=Diffuse_Fresnel_UV_Set, sortPriority=11}
   ) {materialFunction=&MF_Choose_UV_Set}.Result) {
   group=Diffuse_Fresnel
   parameterName=FresnelDiffuse_Texture
   samplerType=SAMPLERTYPE_Masks
   sortPriority=10
   texture=&T_MidgreyMask
  }
  n34 = VectorParameter() {group=Diffuse_Fresnel, parameterName=DiffuseFresnelColor, sortPriority=9, defaultValue=<rgba 1 1 1 1>}
  n35 = StaticBoolParameter() {group=Diffuse_Fresnel, parameterName=Use_Texture_Fresnel, sortPriority=8}
  n36 = MaterialFunctionCall(
   In=Fresnel(
    ExponentIn=ScalarParameter() {defaultValue=3, group=Diffuse_Fresnel, parameterName=Diffuse_Fresnel_Strength, sortPriority=42}
    BaseReflectFractionIn=ScalarParameter() {defaultValue=0.002, group=Diffuse_Fresnel, parameterName=Diffuse_Fresnel_Falloff, sortPriority=43}
   ) {baseReflectFraction=0.003}
   Contrast=ScalarParameter() {defaultValue=0.03, group=Diffuse_Fresnel, parameterName=Disffuse_Fresnel_Contrast, sortPriority=44}
  ) {materialFunction=&CheapContrast}
  n37 = NamedRerouteDeclaration(None=StaticSwitch(
    True=OneMinus(None=@n36.Result)
    False=@n36.Result
    Value=StaticBoolParameter() {group=Diffuse_Fresnel, parameterName=Inner_Diffuse_Fresnel, sortPriority=40}
   )) {name=Fresnel_diffuse, nodeColor=<rgba 0.3828 0 1 1>}
  n38 = ScalarParameter() {defaultValue=1, group=Diffuse_Fresnel, parameterName=Diffuse_Fresnel_Opacity, sliderMax=1, sortPriority=41}
  StaticSwitch(
   True=LinearInterpolate(A=@n32, B=StaticSwitch(True=@n33.RGB, False=@n34.RGB, Value=@n35), Alpha=Multiply(A=@n37, B=@n38))
   False=@n32
   Value=StaticBoolParameter() {group=Diffuse_Fresnel, parameterName=Use_Diffuse_Fresnel, sortPriority=5}
  )
 ]
}
